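# Unlit masked master material for Niagara sprite, ribbon and mesh particles, drawing an outlined circle from UVs.
# An inner sphere mask blends OutLineColor into tinted particle color, with OutLineThickness setting the rim inside MaskRadius.
# The outer sphere mask shapes the circle, optionally multiplied by a blue-noise dither of particle alpha via UseDitheredMask.
# UseDitheredFadeOnly swaps the Opacity-scaled dither for one that turns solid as alpha nears one, ignoring Opacity.
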Material /WildEstate/Environment/TestContent/FX/M_WildEstate_FXMaster
graph {
 properties {
  bAutomaticallySetUsageInEditor=false
  bUsedWithNiagaraMeshParticles=true
  bUsedWithNiagaraRibbons=true
  bUsedWithNiagaraSprites=true
  blendMode=BLEND_Masked
  shadingModel=MSM_Unlit
  usedShadingModels=Unlit
 }
 nodes [
  n1 = ParticleColor()
  n2 = TextureCoordinate()
  n3 = Constant2Vector() {g=0.5, r=0.5}
  n4 = ScalarParameter() {defaultValue=0.5, parameterName=MaskRadius}
  n5 = ScalarParameter() {defaultValue=0.06, parameterName=OutLineThickness}
  LinearInterpolate(
   A=VectorParameter() {parameterName=OutLineColor, defaultValue=<rgba 0 0 0 1>}.RGB
   B=Multiply(A=VectorParameter() {parameterName=Tint, defaultValue=<rgba 1 1 1 1>}.RGB, B=@n1.RGB)
   Alpha=SphereMask(A=@n2, B=@n3, Radius=Subtract(A=@n4, B=@n5))
  )
  n6 = SphereMask(A=@n2, B=@n3, Radius=@n4)
  n7 = ScalarBlueNoise()
  n8 = LinearInterpolate(A=LinearInterpolate(B=@n1.A, Alpha=@n7), Alpha=@n1.A)
  n9 = LinearInterpolate(B=Multiply(A=@n1.A, B=ScalarParameter() {defaultValue=1, parameterName=Opacity}), Alpha=@n7)
  StaticSwitchParameter(True=Multiply(A=@n6, B=StaticSwitchParameter(True=@n8, False=@n9) {parameterName=UseDitheredFadeOnly}), False=@n6) {parameterName=UseDitheredMask}
 ]
}
