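# M_StylizedWall_Mat is a material in the environment test content, named for a stylized wall.
# Its graph is empty, with no expressions and no connected inputs shown, so it defines no stylized look of its own.
# It uses the Default Lit shading model and has automatic setting of usage flags in the editor turned off.
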Material /WildEstate/Environment/TestContent/M_StylizedWall_Mat
graph {expressions {}, properties {bAutomaticallySetUsageInEditor=false, usedShadingModels=Default Lit}}
